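# SoundCue for the Spicy Soda (SmallHealthSpeed item) boost loop, routed through the consumables sound class with consumable-use attenuation.
# A Mixer layers a looping boost wave, weighted louder, with a one-shot pepper alarm that starts after a short fixed delay.
# An Enveloper on the loop fades it in quickly, holds full volume, then tapers it to a low level over several seconds.
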SoundCue /FNGameplayCues/Athena/Sounds/Items/SmallHealthSpeed/SC_SpicySoda_Boost_Loop
root=Mixer_0
nodes [
 {id=Delay_0, children [WavePlayer_1], properties {delayMax=0.3, delayMin=0.3}}
 {
  id=Enveloper_0
  children [WavePlayer_0]
  properties {pitchCurve {editorCurveData {keys [{time=0}]}}, volumeCurve {editorCurveData {keys [<at 0 0.0017>,<at 0.5305 1>,<at 3 1>,<at 6 0.2>]}}}
 }
 {id=Mixer_0, children [Enveloper_0,Delay_0], properties {inputVolume [1.5,0.7]}}
 {id=WavePlayer_0, properties {bLooping=true, soundWaveAssetPtr=&/SpicySoda_Boost_Loop_New}}
 {id=WavePlayer_1, properties {soundWaveAssetPtr=&/Athena_Pepper_Alarm_New}}
]
properties {
 attenuationSettings=&Fort_Consumable_Use_Athena_Att
 maxDistance=2250
 soundClassObject=&Fort_Consumables
 volumeMultiplier=0.45
}
